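# Material function for the DryErase keytar cosmetic that adds an animated screen face to incoming BaseColor and Emissive.
# Curve-atlas rows drive flipbook frames for idle, happy and mad moods picked by IdleIndex, with a symbol overlay.
# IsVariant swaps in a glitch-offset Toro flipbook.
# Bump-offset parallax, masked shadows and pulsing tinted emissive finish the screen, with timing from music-clock collection parameters or time.
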MaterialFunction /SparksCosmetics/Instrument/Keytar/DryErase/Materials/MF_DryErase_Keytar
graph {nodes [
  n1 = NamedRerouteUsage()
  n2 = Constant() {r=1}
  n3 = LinearInterpolate(A=Saturate(None=Fmod(A=@n1, B=@n2)), B=NamedRerouteUsage(), Alpha=NamedRerouteUsage())
  n4 = Add(A=@n3) {constB=0.2}
  n5 = Constant() {r=3.5}
  n6 = Round(None=Multiply(A=Cosine(None=@n4), B=@n5))
  Add(B=Multiply(A=Divide(A=@n6, B=@n5)) {constB=0.01})
  n7 = Constant() {r=3}
  Divide(A=Round(None=Multiply(B=@n7)), B=@n7)
  n8 = NamedRerouteUsage()
  n9 = ScalarParameter() {defaultValue=0.7, group=DryErase, parameterName=ShadowColorMulti}
  n10 = NamedRerouteUsage()
  n11 = LinearInterpolate(A=@n8, B=Multiply(A=@n8, B=LinearInterpolate(A=@n9, Alpha=@n10)), Alpha=NamedRerouteUsage())
  n12 = LinearInterpolate(A=@n11, Alpha=Abs(None=NamedRerouteUsage())) {constAlpha=1, constB=0}
  n13 = NamedRerouteUsage()
  n14 = Multiply(A=NamedRerouteUsage(), B=ScalarParameter() {defaultValue=0.5, group=DryErase, parameterName=ScreenBaseColor_Intensity})
  n15 = ScalarParameter() {defaultValue=0.2, group=DryErase, parameterName=ScreenBaseColor_Desaturation}
  FunctionOutput(None=LinearInterpolate(
    A=LinearInterpolate(A=@n11, B=@n12, Alpha=@n13) {constAlpha=0, constB=-0.45}
    B=Desaturation(None=@n14, Fraction=@n15)
    Alpha=NamedRerouteUsage()
   )) {outputName=BaseColor}
  n16 = Constant()
  FunctionOutput(None=Add(A=FunctionInput(Preview=@n16) {bUsePreviewValueAsDefault=true, inputName=Emissive, sortPriority=1}, B=NamedRerouteUsage())) {outputName=Emissive, sortPriority=1}
  If(
   A=Constant() {r=140}
   B=NamedRerouteDeclaration(None=CollectionParameter() {collection=/Game/Sounds/MusicEventSubsystem/MPC_MusicEventSubsystem_PriorityClock, parameterName=Tempo}) {name=BPM, nodeColor=<rgba 1 0.7734 0 1>}
   "A > B"=Constant() {r=1}
   "A < B"=Constant() {r=2}
  ) {constB=60}
  n17 = VectorParameter() {group=DryErase, parameterName=ScreenColor, defaultValue=<rgba 0 0.7655 1 1>}
  MaterialFunctionCall(
   Hue Shift Percentage=LinearInterpolate(Alpha=ScalarParameter() {group=DryErase, parameterName=ShiftToOrange}) {constAlpha=0, constB=-0.45}
   Texture=@n17.RGB
  ) {materialFunction=&HueShift}
  n18 = Constant() {r=2}
  n19 = CurveAtlasRowParameter(CurveTime=LinearInterpolate(A=NamedRerouteUsage(), B=NamedRerouteUsage(), Alpha=NamedRerouteUsage())) {
   atlas=&/CA_DryErase_Keytar
   curve=&/CLC_DryErase_Keytar_HappySymbols
   defaultValue=3
   group=DryErase
   parameterName=Symbols_CurveAtlas
   sortPriority=1
  }
  n20 = MaterialFunctionCall!(
   UVs=BumpOffset(
    Coordinate=TextureCoordinate() {coordinateIndex=1}
    Height=ScalarParameter() {defaultValue=-3, group=DryErase, parameterName=B_BumpOffsetHeight, sortPriority=-1}
   )
   OffsetScale=ScalarParameter() {defaultValue=0.7, group=DryErase, parameterName=B_OffsetScale, sortPriority=-1}
   OffsetRotation=ScalarParameter() {group=DryErase, parameterName=B_OffsetR, sortPriority=-1}
   OffsetU=ScalarParameter() {defaultValue=0.05, group=DryErase, parameterName=B_OffsetU, sortPriority=-1}
   OffsetV=ScalarParameter() {defaultValue=-0.15, group=DryErase, parameterName=B_OffsetV, sortPriority=-1}
   Number of Columns=@n18
   Index=@n19
   Number of Rows=@n18
  ) {materialFunction=/SparksCosmetics/Instrument/Keytar/ChiveFlake/Materials/MF_ChiveFlake_Keytar_Flipbook}
  NamedRerouteDeclaration(None=@n20.Coordinates) {name=B_Coordinates, nodeColor=<rgba 1 0 0.7266 1>}
  NamedRerouteDeclaration(None=@n20.Mask) {name=B_Mask, nodeColor=<rgba 1 0 0.7266 1>}
  NamedRerouteDeclaration(None=FunctionInput(Preview=Constant() {r=0.2}) {bUsePreviewValueAsDefault=true, inputName=BaseColor}) {name=BaseColor, nodeColor=<rgba 0 0.8359 1 1>}
  n21 = TextureCoordinate() {coordinateIndex=1}
  n22 = Constant() {r=0.5}
  n23 = ScalarParameter() {defaultValue=1.2, group=DryErase, parameterName=Variant_OffsetScale, sortPriority=2}
  n24 = Divide(A=Subtract(A=Multiply(A=@n21), B=@n22), B=AppendVector(A=@n23, B=Divide(A=@n23) {constB=4}))
  n25 = Constant() {r=16}
  n26 = TextureCoordinate() {coordinateIndex=1}
  n27 = Subtract(A=Multiply(A=MaterialFunctionCall!(UV=@n26) {materialFunction=/Game/Weapons/WeaponSkins/MaterialFunction/MF_GlitchFunction}.Glitched UV) {constB=0.06}) {constB=0.05}
  n28 = MaterialFunctionCall(
   UVs=BumpOffset(Coordinate=Add(A=@n24, B=@n22), Height=ScalarParameter() {defaultValue=-3, group=DryErase, parameterName=Variant_BumpOffsetHeight})
   OffsetScale=Constant() {r=1}
   OffsetRotation=ScalarParameter() {defaultValue=-90, group=DryErase, parameterName=Variant_OffsetRotation, sortPriority=2}
   OffsetU=Add(A=ScalarParameter() {group=DryErase, parameterName=Variant_OffsetU, sortPriority=2}, B=ComponentMask(None=@n27) {r=true})
   OffsetV=Add(A=ScalarParameter() {group=DryErase, parameterName=Variant_OffsetV, sortPriority=3}, B=ComponentMask(None=@n27) {g=true})
   Number of Columns=Constant() {r=2}
   Index=Fmod(A=Multiply(A=@n3, B=@n25) {constB=16}, B=@n25)
   Number of Rows=Constant() {r=8}
  ) {materialFunction=&MF_ChiveFlake_Keytar_Flipbook}
  n29 = TextureSample(UVs=@n28.Coordinates) {texture=/SparksCosmetics/Instrument/Keytar/DryErase/Textures/T_DryEraseToro_Keytar_Anim_FB}
  n30 = Multiply(A=@n29.A, B=@n28.Mask)
  NamedRerouteDeclaration(None=LinearInterpolate(A=@n29.RGB, B=NamedRerouteUsage(), Alpha=OneMinus(None=@n30)) {constB=0.2}) {name=Variant_Color, nodeColor=<rgba 1 0 0.7266 1>}
  NamedRerouteDeclaration(None=Saturate(None=@n30)) {name=Variant_Mask, nodeColor=<rgba 1 0 0.7266 1>}
  n31 = TextureCoordinate()
  n32 = ScalarParameter() {defaultValue=-0.2, group=DryErase, parameterName=Shadow_BumpOffsetHeight}
  n33 = BumpOffset(Coordinate=@n31, Height=@n32)
  n34 = TextureObjectParameter() {
   group=DryErase
   parameterName=ScreenMask
   samplerType=SAMPLERTYPE_Masks
   texture=/SparksCosmetics/Instrument/Keytar/DryErase/Textures/T_DryErase_Keytar_Screen_FX
  }
  NamedRerouteDeclaration(None=SmoothStep(
    Max=ScalarParameter() {defaultValue=0.7, group=DryErase, parameterName=ShadowMax}
    Value=TextureSample(UVs=@n33, Tex=@n34) {samplerSource=SSM_Wrap_WorldGroupSettings}.G
   ) {constMax=0.3}) {name=ShadowArea, nodeColor=<rgba 0 1 0.3906 1>}
  NamedRerouteDeclaration(None=CollectionParameter() {collection=&MPC_MusicEventSubsystem_PriorityClock, parameterName=MusicTimestampValid}) {name=IsMusicOn, nodeColor=<rgba 1 0.0408 0 1>}
  NamedRerouteDeclaration(None=StaticSwitch(True=ScalarParameter() {group=DryErase, parameterName=LocalTime}, False=Time())) {name=Time, nodeColor=<rgba 0.3372 0.0176 0.0168 1>}
  n35 = Constant() {r=1}
  n36 = Divide(
   A=NamedRerouteDeclaration(None=CollectionParameter() {collection=&MPC_MusicEventSubsystem_PriorityClock, parameterName=MeasureBeat}) {name=MusicBeat, nodeColor=<rgba 1 0 0.0469 1>}
   B=@n35
  )
  NamedRerouteDeclaration(None=Saturate(None=Fmod(A=@n36, B=@n35))) {name=MusicTime, nodeColor=<rgba 1 0 0.0469 1>}
  n37 = CurveAtlasRowParameter(CurveTime=Subtract(A=NamedRerouteUsage(), B=ScalarParameter() {defaultValue=-10, group=DryErase, parameterName=HitShake}) {constB=0}) {
   atlas=&CA_DryErase_Keytar
   curve=&/CLC_DryErase_Keytar_EmissivePulse
   defaultValue=4
   group=Flipbook General
   parameterName=Emissive_CurveAtlas
  }
  n38 = NamedRerouteDeclaration(None=Subtract(A=@n37.B)) {name=HitShake, nodeColor=<rgba 0 1 0.9297 1>}
  n39 = Constant() {r=4}
  n40 = MaterialFunctionCall(
   UVs=BumpOffset(
    Coordinate=TextureCoordinate() {coordinateIndex=1}
    Height=ScalarParameter() {defaultValue=-3, group=DryErase, parameterName=A_BumpOffsetHeight, sortPriority=-2}
   )
   OffsetScale=ScalarParameter() {defaultValue=0.8, group=DryErase, parameterName=A_OffsetScale, sortPriority=-2}
   OffsetRotation=ScalarParameter() {group=DryErase, parameterName=A_OffsetR, sortPriority=-2}
   OffsetU=ScalarParameter() {group=DryErase, parameterName=A_OffsetU, sortPriority=-2}
   OffsetV=Add(A=ScalarParameter() {defaultValue=0.1, group=DryErase, parameterName=A_OffsetV, sortPriority=-2}, B=Multiply(A=@n38) {constB=0.05}) {constB=0.05}
   Number of Columns=@n39
   Index=ComponentMask(None=NamedRerouteUsage()) {r=true}
   Number of Rows=@n39
  ) {materialFunction=&MF_ChiveFlake_Keytar_Flipbook}
  NamedRerouteDeclaration(None=@n40.Coordinates) {name=A_Coordinates, nodeColor=<rgba 1 0 0.7266 1>}
  NamedRerouteDeclaration(None=@n40.Mask) {name=A_Mask, nodeColor=<rgba 1 0 0.7266 1>}
  NamedRerouteDeclaration(None=@n32) {name=OffsetHeight, nodeColor=<rgba 0.1016 0 1 1>}
  n41 = TextureSample(UVs=NamedRerouteUsage()) {texture=/SparksCosmetics/Instrument/Keytar/DryErase/Textures/T_DryErase_Keytar_Anim_FB}
  n42 = Multiply(A=@n41.A, B=NamedRerouteUsage())
  n43 = LinearInterpolate(A=@n41.RGB, B=NamedRerouteUsage(), Alpha=OneMinus(None=@n42)) {constB=0.2}
  n44 = TextureSample(UVs=NamedRerouteUsage()) {texture=/SparksCosmetics/Instrument/Keytar/DryErase/Textures/T_DryErase_Keytar_AnimSymbols_FB}
  n45 = OneMinus(None=ComponentMask(None=NamedRerouteUsage()) {b=true})
  n46 = Multiply(A=@n44.A, B=LinearInterpolate(A=NamedRerouteUsage(), Alpha=Saturate(None=@n45)) {constB=0})
  NamedRerouteDeclaration(None=StaticSwitchParameter(True=NamedRerouteUsage(), False=LinearInterpolate(A=@n43, B=@n44.RGB, Alpha=@n46)) {group=DryErase, parameterName=IsVariant}) {name=IdleAnimation, nodeColor=<rgba 1 0.7734 0 1>}
  NamedRerouteDeclaration(None=TextureSample(UVs=@n31, Tex=@n34) {samplerSource=SSM_Wrap_WorldGroupSettings}.R) {name=AreaMask, nodeColor=<rgba 0 1 0.3203 1>}
  n47 = Add(A=@n42, B=@n46)
  NamedRerouteDeclaration(None=StaticSwitchParameter(True=NamedRerouteUsage(), False=Saturate(None=@n47)) {group=DryErase, parameterName=IsVariant}) {name=IdleAnimationAlpha, nodeColor=<rgba 0 1 0.1328 1>}
  n48 = NamedRerouteUsage()
  n49 = LinearInterpolate(
   A=@n17.RGB
   B=ComponentMask(None=Constant4Vector() {constant=<rgba 1 0.3631 0.0414 1>}.RGBA) {b=true, g=true, r=true}
   Alpha=Abs(None=NamedRerouteUsage())
  ) {constAlpha=0, constB=-0.45}
  n50 = LinearInterpolate(A=Multiply(A=@n49) {constB=0.5}, B=@n49, Alpha=NamedRerouteUsage())
  n51 = Saturate(None=Subtract(A=NamedRerouteUsage(), B=NamedRerouteUsage()))
  n52 = Add(
   A=Multiply(A=LinearInterpolate(A=@n48, B=@n50, Alpha=@n51), B=ScalarParameter() {defaultValue=0.5, group=DryErase, parameterName=EmissiveIntensity}) {constB=0}
   B=Multiply(A=Fresnel()) {constB=0.1}
  )
  n53 = NamedRerouteUsage()
  n54 = NamedRerouteUsage()
  n55 = Saturate(None=Fmod(A=Multiply(A=@n54) {constB=2}, B=Constant() {r=1}))
  n56 = NamedRerouteUsage()
  n57 = CurveAtlasRowParameter(CurveTime=LinearInterpolate(A=StaticSwitchParameter(True=@n55, False=@n56) {group=DryErase, parameterName=IsVariant}, B=NamedRerouteUsage(), Alpha=NamedRerouteUsage())) {
   atlas=&CA_DryErase_Keytar
   curve=&CLC_DryErase_Keytar_EmissivePulse
   defaultValue=4
   group=Flipbook General
   parameterName=Emissive_CurveAtlas
  }
  n58 = LinearInterpolate(A=Floor(None=NamedRerouteUsage()), Alpha=NamedRerouteUsage())
  n59 = LinearInterpolate(
   A=LinearInterpolate(B=@n57.R, Alpha=Saturate(None=@n58)) {constAlpha=1, constB=0}
   B=StaticSwitchParameter(True=Constant(), False=NamedRerouteUsage()) {group=DryErase, parameterName=IsVariant}
   Alpha=Saturate(None=Subtract(A=@n58))
  ) {constAlpha=0, constB=0}
  n60 = StaticSwitchParameter(True=LinearInterpolate(Alpha=NamedRerouteUsage()) {constA=0.3, constB=0}, False=Constant()) {group=DryErase, parameterName=IsVariant}
  n61 = Constant() {r=1}
  n62 = ScalarParameter() {defaultValue=0.3, group=DryErase, parameterName=BaseEmissiveOpacity}
  NamedRerouteDeclaration(None=Multiply(
    A=Multiply(A=@n52, B=@n53)
    B=MaterialFunctionCall(Input=@n59, Input Low=@n60, Input High=@n61, Target Low=@n62, Target High=@n61) {materialFunction=&RemapValueRange}.Result
   )) {name=AdditionalEM, nodeColor=<rgba 1 0.6562 0 1>}
  n63 = CurveAtlasRowParameter() {
   atlas=&CA_DryErase_Keytar
   curve=&/CLC_DryErase_Keytar_Idle
   group=DryErase
   parameterName=A_IdleCurveAtlas
   sortPriority=-2
  }
  n64 = Fmod(A=Divide(A=NamedRerouteUsage(), B=@n63.A) {constB=1.1, desc=time it lasts}, B=Constant() {r=1})
  n65 = CurveAtlasRowParameter(CurveTime=LinearInterpolate(A=Saturate(None=@n64), B=NamedRerouteUsage(), Alpha=NamedRerouteUsage())) {
   atlas=&CA_DryErase_Keytar
   curve=&CLC_DryErase_Keytar_Idle
   group=DryErase
   parameterName=A_IdleCurveAtlas
   sortPriority=0
  }
  n66 = NamedRerouteUsage()
  n67 = CurveAtlasRowParameter() {
   atlas=&CA_DryErase_Keytar
   curve=&/CLC_DryErase_Keytar_Happy
   defaultValue=2
   group=DryErase
   parameterName=B_IdleCurveAtlas
   sortPriority=-2
  }
  n68 = Saturate(None=Fmod(A=Divide(A=@n66, B=@n67.A) {constB=1.1, desc=time it lasts}, B=Constant() {r=1}))
  n69 = CurveAtlasRowParameter(CurveTime=LinearInterpolate(A=NamedRerouteDeclaration(None=@n68) {name=HappyLerp, nodeColor=<rgba 0.9922 0 1 1>}, B=NamedRerouteUsage(), Alpha=NamedRerouteUsage())) {
   atlas=&CA_DryErase_Keytar
   curve=&CLC_DryErase_Keytar_Happy
   defaultValue=2
   group=DryErase
   parameterName=B_IdleCurveAtlas
   sortPriority=0
  }
  n70 = Floor(None=NamedRerouteDeclaration(None=ScalarParameter() {group=DryErase, parameterName=IdleIndex, sliderMax=2}) {name=CurrentIldeIndex, nodeColor=<rgba 0.7812 1 0 1>})
  n71 = Saturate(None=@n70)
  n72 = CurveAtlasRowParameter() {
   atlas=&CA_DryErase_Keytar
   curve=&/CLC_DryErase_Keytar_Mad
   defaultValue=1
   group=DryErase
   parameterName=C_IdleCurveAtlas
   sortPriority=-2
  }
  n73 = Fmod(A=Divide(A=NamedRerouteUsage(), B=@n72.A) {constB=1.1, desc=time it lasts}, B=Constant() {r=1})
  n74 = CurveAtlasRowParameter(CurveTime=LinearInterpolate(A=Saturate(None=@n73), B=NamedRerouteUsage(), Alpha=NamedRerouteUsage())) {
   atlas=&CA_DryErase_Keytar
   curve=&CLC_DryErase_Keytar_Mad
   defaultValue=1
   group=DryErase
   parameterName=C_IdleCurveAtlas
   sortPriority=0
  }
  n75 = Subtract(A=@n70)
  NamedRerouteDeclaration(None=LinearInterpolate(A=LinearInterpolate(A=@n65, B=@n69, Alpha=@n71) {constAlpha=1}, B=@n74, Alpha=Saturate(None=@n75)) {constAlpha=0}) {name=AnimationIndex, nodeColor=<rgba 0.6406 0 1 1>}
  n76 = LinearInterpolate(A=NamedRerouteUsage(), B=NamedRerouteUsage(), Alpha=NamedRerouteUsage())
  n77 = Constant() {r=2}
  n78 = CurveAtlasRowParameter(CurveTime=Saturate(None=Fmod(A=@n76, B=@n77))) {
   atlas=&CA_DryErase_Keytar
   curve=&CLC_DryErase_Keytar_EmissivePulse
   defaultValue=4
   group=Flipbook General
   parameterName=Emissive_CurveAtlas
  }
  n79 = AppendVector(A=@n78.G, B=@n78.G)
  StaticSwitch(True=Multiply(A=Subtract(A=@n79), B=ScalarParameter() {defaultValue=0.1, group=DryErase, parameterName=Variant_GlitchIntensity}) {constB=0.2})
  Time() {bOverride_Period=true, period=1}
 ]}
