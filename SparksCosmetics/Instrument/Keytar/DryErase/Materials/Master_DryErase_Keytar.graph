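# Master material for the Dry Erase keytar instrument cosmetic, built for skeletal meshes on the 2023 character parent material pipeline.
# MF_CharacterParent_2023_PreFX supplies material attributes, and MF_DryErase_Keytar processes the extracted color.
# The result is written back into the attributes and passed to MF_CharacterParent_2023_PostFX.
# Shading model comes from the material expression, with pixel animation and mesh deformer usage enabled.
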Material /SparksCosmetics/Instrument/Keytar/DryErase/Materials/Master_DryErase_Keytar
graph {
 properties {
  bAutomaticallySetUsageInEditor=false
  bHasPixelAnimation=true
  bUseMaterialAttributes=true
  bUsedWithMeshDeformer=true
  bUsedWithSkeletalMesh=true
  pixelDepthOffsetMode=PDOM_Legacy
  refractionMethod=RM_IndexOfRefraction
  shadingModel=MSM_FromMaterialExpression
  subsurfaceProfile=&SS_HeroSkin_02
  usedShadingModels=Default Lit | Subsurface | Clear Coat | Subsurface Profile | Cloth
 }
 nodes [
  n1 = MaterialFunctionCall() {materialFunction=&MF_CharacterParent_2023_PreFX}
  n2 = GetMaterialAttributes(None=@n1.Result) {attributeGetTypes [#1, #2, #3]}
  n3 = MaterialFunctionCall!(BaseColor=@n2.Base Color, Emissive=@n2.Base Color) {materialFunction=&/MF_DryErase_Keytar}
  MaterialFunctionCall(
   Material=SetMaterialAttributes!(MaterialAttributes=@n2.MaterialAttributes, Base Color=@n3.BaseColor, Emissive Color=@n3.BaseColor, Specular=@n2.MaterialAttributes) {attributeSetTypes [#1, #2, #3]}
   M=@n1.M
   Pre Skinned Local Position=@n1.M
   Distance Blend=@n1.M
   Fresnel=@n1.M
  ) {materialFunction=&MF_CharacterParent_2023_PostFX}
 ]
}
