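# Physics asset for the DryErase Keytar cosmetic instrument in SparksCosmetics, covering a chain of rings and links attached to its handle.
# Capsule bodies cover the handle (two capsules), a main ring, five links and a zero-length, sphere-like end ring.
# Constraints join them in series from handle to end ring with per-joint swing limits.
# Twist is locked on the four joints nearest the handle, but not on the last three.
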PhysicsAsset /SparksCosmetics/Instrument/Keytar/DryErase/Meshes/DryErase_Keytar_Physics
bodies {
 handle [
  {shapeType=Capsule, length=20.6179, radius=5.663, center=<0 -7.5574 -4.2101>}
  {shapeType=Capsule, length=15.5678, radius=2.9411, shapeName=handle_capsule, center=<1.347 -19.9767 4.4199>, rotation=<rot 0 0 -4.6813>}
 ]
 link_01 [{
   shapeType=Capsule
   length=1.8166
   radius=1.0944
   shapeName=link_01_capsule
   center=<1.0784 -0.1261 -0.0951>
   rotation=<rot 0 -99.7846 -89.2802>
  }]
 link_02 [{
   shapeType=Capsule
   length=1.8168
   radius=1.0819
   shapeName=link_02_capsule
   center=<1.0607 -0.0742 0.0252>
   rotation=<rot 0 80.2323 -84.3502>
  }]
 link_03 [{
   shapeType=Capsule
   length=1.8162
   radius=1.0891
   shapeName=link_03_capsule
   center=<1.0223 -0.1226 -0.0415>
   rotation=<rot 0 80.288 -87.9313>
  }]
 link_04 [{
   shapeType=Capsule
   length=1.8151
   radius=1.0912
   shapeName=link_04_capsule
   center=<0.9403 -0.0421 -0.0666>
   rotation=<rot 0 80.2422 -88.7015>
  }]
 link_05 [{shapeType=Capsule, length=1.505, radius=1.005, shapeName=link_05_capsule, center=<1.0335 -0.003 -0.0699>, rotation=<rot 0 -88.2893 -87.5094>}]
 ring_end [{shapeType=Capsule, length=0, radius=4.4339, shapeName=ring_end_capsule, center=<3.3179 0.0199 -0.4456>}]
 ring_main [{shapeType=Capsule, length=0, radius=3.6465, shapeName=ring_main_capsule, rotation=<rot 0 0 90>}]
}
constraints [
 {bone1Name=ring_main, bone2Name=handle, swing1LimitDegrees=10, swing2LimitDegrees=55, twistLimitDegrees=45, twistMotion=Locked}
 {bone1Name=link_01, bone2Name=ring_main, swing1LimitDegrees=30, swing2LimitDegrees=15, twistLimitDegrees=45, twistMotion=Locked}
 {bone1Name=link_02, bone2Name=link_01, swing1LimitDegrees=30, swing2LimitDegrees=15, twistLimitDegrees=45, twistMotion=Locked}
 {bone1Name=link_03, bone2Name=link_02, swing1LimitDegrees=20, swing2LimitDegrees=20, twistLimitDegrees=45, twistMotion=Locked}
 {bone1Name=link_04, bone2Name=link_03, swing1LimitDegrees=30, swing2LimitDegrees=30, twistLimitDegrees=3}
 {bone1Name=link_05, bone2Name=link_04, swing1LimitDegrees=30, swing2LimitDegrees=30, twistLimitDegrees=4}
 {bone1Name=ring_end, bone2Name=link_05, swing1LimitDegrees=20, swing2LimitDegrees=40, twistLimitDegrees=5}
]
