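# Animation Blueprint for the DryErase keytar cosmetic, adding rigid-body physics simulated relative to the handle bone and driving reactive material states.
# Each update selects the material's IdleIndex from gameplay-event state, pickaxe-style swing delegates and music clock validity, while previews cycle states on timers.
# Eliminations, chests and health or shield changes set temporary states, with damage also stamping game time into a HitShake parameter.
# Skydiving and parachute status set a GravityOverride variable.
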AnimBlueprint /SparksCosmetics/Instrument/Keytar/DryErase/Meshes/DryErase_Keytar_AnimBP
target_skeleton=&/DryErase_Keytar_Skeleton
graphs [
 AnimGraph {nodes [
   n1 = RigidBody(ComponentPose=Local To Component(LocalPose=Input Pose())) {
    bOverrideWorldGravity=true
    overridePhysicsAsset=&/DryErase_Keytar_Physics
    simulationSpace=BaseBoneSpace
    alphaBoolBlend {blendOption=Linear}
    baseBoneRef {boneName=handle}
    componentAppliedLinearAccClamp=<3000 3000 3000>
    componentLinearAccScale=<0.4 0.4 0.4>
    componentLinearVelScale=<0.2 0.2 0.2>
    overrideWorldGravity {z=-980}
    simSpaceSettings {
     maxAngularAcceleration=1000
     maxAngularVelocity=6
     maxLinearAcceleration=700
     maxLinearVelocity=600
     velocityScaleZ=0
     worldAlpha=0.6
     externalLinearDragV=<0.2 0.2 0.02>
    }
   }
   Output Pose(Result=Component To Local(ComponentPose=@n1))
  ]}
 EventGraph {nodes [
   StartSwinging() {is=title K2Node_CustomEvent}
   > Set Is Swinging(Is Swinging:bool=true) -> (Output_Get:bool)
   > RetriggerableDelay(Duration=1.5)
   > Set Is Swinging(Is Swinging=false)
   Event BlueprintBeginPlay() {is=title K2Node_Event}
   > n1 = Cast To PlayerPawn_Athena(Object=TryGetPawnOwner())
   n2 = Create Event()
   > Bind Event to Swing Left(self:Player Pawn Athena Object Reference=@n1, Delegate=@n2) {is=title K2Node_AddDelegate}
   > Bind Event to Swing Right(self:Player Pawn Athena Object Reference=@n1, Delegate=@n2) {is=title K2Node_AddDelegate}
   > Set IsPreview(IsPreview:bool=IsCosmeticObjectDisplayedAsPreview!(CosmeticObject:Object Reference=GetOwningActor()) -> (ReturnValue:bool)) -> (Output_Get:bool)
   Event BlueprintUpdateAnimation()
   > n3 = Sequence()
   n4 = Branch(execute=@n3.then_1, Condition=Get IsPreview() -> (IsPreview:bool))
   > n5 = Do Once()
   n6 = MultiGate!(execute:Exec=[@n5.Completed, @n8.then, @n10.then], Reset:Exec, IsRandom:bool, Loop:bool=true, StartIndex:int=3) -> (Out 0:Exec, Out 1:Exec, Out 2:Exec, Out 3:Exec) {is=title K2Node_MultiGate}
   > n7 = Set State(State:int=1) -> (Output_Get:int=0)
   SetIdle!(execute=[@n11.then, @n7.then], self, InInt:int=Get State() -> (State:int=0)) -> (then)
   > n8 = Delay(Duration=3)
   n9 = Set State(execute=[@n6.Out 1, @n6.Out 3], State=0) -> (Output_Get=0)
   > SetIdle(InInt=@n9)
   > n10 = Delay(Duration=4.5)
   Set State(execute=@n6.Out 2, State=2) -> (Output_Get=0)
   > n11 = HitShake()
   n12 = GetHighestPriorityClock!(self:Music Event Subsystem Object Reference=Get MusicEventSubsystem() -> (ReturnValue:Music Event Subsystem Object Reference) {is=title K2Node_GetSubsystem}) -> (ReturnValue:Music Clock Object Reference)
   n13 = Is Valid(exec=@n4.else, InputObject=@n12)
   n14 = Get Is Swinging() -> (Is Swinging:bool)
   n15 = GetSongPos!(self:Music Clock Object Reference=@n12, Timebase:ECalibratedMusicTimebase Enum=VideoRenderTime) -> (ReturnValue:"MIDI Song Position Structure (by ref)")
   n16 = Get State() -> (State=0)
   > n17 = Branch(Condition=AND Boolean!(
     A=Is MIDI Song Position Valid!(SongPos:"MIDI Song Position Structure (by ref)"=@n15) -> (ReturnValue:bool)
     B="Not Equal (Integer)"(A=@n16, B=2)
     C:bool=NOT Boolean(A=@n14)
    ) -> (ReturnValue))
   > SetIdle(InInt=1)
   SetIdle(execute=[@n17.else, @n13.Is Not Valid], InInt=SelectInt(A=2, B=Get State() -> (State=0), bPickA=Get Is Swinging()))
   n18 = Branch(execute=@n3.then_0, Condition=Get bIsSkyDiving())
   > n19 = Branch(Condition=Get bIsParachuteOpened())
   > Set GravityOverride(GravityOverride="0,0,-400")
   Set GravityOverride(execute=@n19.else, GravityOverride="0,0,10000")
   Set GravityOverride(execute=@n18.else, GravityOverride="0,0,-980")
   n20 = Event BlueprintProcessAnimInputEvent()
   > n21 = Branch(Condition=Get IsPreview())
   n22 = Switch on Name!(
    execute:Exec=@n21.else
    Selection:Name=Get EventName!(self:Fort Anim Input Event Object Reference=@n20.AnimInputEvent) -> (EventName:Name=None)
   ) -> (
    aie.EnemyPlayerEliminated:Exec
    aie.ChestOpened:Exec
    aie.PlayerVitality.ShieldsIncreased:Exec
    aie.PlayerVitality.HealthIncreased:Exec
    aie.PlayerVitality.ShieldsDecreased:Exec
    aie.PlayerVitality.HealthDecreased:Exec
   ) {
    is=title K2Node_SwitchName
    bHasDefaultPin=false
    pinNames [
     aie.EnemyPlayerEliminated
     aie.ChestOpened
     aie.PlayerVitality.ShieldsIncreased
     aie.PlayerVitality.HealthIncreased
     aie.PlayerVitality.ShieldsDecreased
     aie.PlayerVitality.HealthDecreased
    ]
   }
   n23 = Set State(execute=[@n22.aie.ChestOpened, @n22.aie.PlayerVitality.HealthIncreased, @n22.aie.PlayerVitality.ShieldsIncreased], State=1) -> (Output_Get=0)
   RetriggerableDelay(execute=[@n23.then, @n24.then], Duration=3)
   > Set State(State=0) -> (Output_Get=0)
   n24 = Set State(execute=[@n25.then, @n22.aie.EnemyPlayerEliminated], State=2) -> (Output_Get=0)
   n25 = HitShake(execute=[@n22.aie.PlayerVitality.HealthDecreased, @n22.aie.PlayerVitality.ShieldsDecreased])
  ]}
 HitShake {nodes [
   HitShake() {is=title K2Node_FunctionEntry}
   > n1 = Is Valid(InputObject=Get MatInst() -> (MatInst:Material Instance Dynamic Object Reference))
   n2 = GetOwningComponent()
   Is Valid(exec=@n1.Is Not Valid, InputObject=@n2)
   > n3 = Get Dynamic Material Instance(SkeletalMeshComponent=@n2)
   > n4 = Set MatInst(MatInst:Material Instance Dynamic Object Reference=@n3) -> (Output_Get:Material Instance Dynamic Object Reference)
   SetScalarParameterValue(execute=[@n1.Is Valid, @n4.then], self=@n4, ParameterName=HitShake, Value=GetGameTimeInSeconds())
  ]}
 SetIdle {nodes [
   n1 = SetIdle!() -> (then, InInt:int) {is=title K2Node_FunctionEntry}
   > n2 = Is Valid(InputObject=Get MatInst())
   n3 = GetOwningComponent()
   Is Valid(exec=@n2.Is Not Valid, InputObject=@n3)
   > n4 = Get Dynamic Material Instance(SkeletalMeshComponent=@n3)
   > n5 = Set MatInst(MatInst=@n4)
   n6 = Get Scalar Parameter Value!(execute=[@n2.Is Valid, @n5.then], self:Material Instance Dynamic Object Reference=@n5, ParameterName:Name=IdleIndex) -> (then, ReturnValue:float) {is=title K2Node_CallFunction}
   n7 = "To Float (Integer)"(InInt=@n1)
   > Branch(Condition="Not Equal (Float)"(A=@n6, B=@n7))
   > SetScalarParameterValue(self=@n5, ParameterName=IdleIndex, Value=@n7)
  ]}
]
variables [
 GravityOverride {type=Vector}
 Is Swinging {type=boolean}
 IsPreview {type=boolean}
 MatInst {type=MaterialInstanceDynamic}
 RandomSequence {type="array<integer>", default [0,1]}
 State {type=integer}
]
